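# Physics asset for the clown mask head accessory, set up for moving left and right eye pieces.
# A capsule body on the root anchors two thin, rotated box bodies on the dyn_eye_l and dyn_eye_r bones.
# Each eye is constrained to the root with twist and Swing1 locked and Swing2 free, so it rotates like an unlimited single-axis hinge.
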PhysicsAsset /Game/Accessories/Hats/F_MED_Clown_Mask_01/Meshes/F_MED_ClownMask_01_Physics
bodies {
 dyn_eye_l [{shapeType=Box, extentX=1.3476, extentY=5.9948, extentZ=6.1054, center=<-0.1 0.1 0.01>, rotation=<rot 0 89.2053 -80.0931>}]
 dyn_eye_r [{shapeType=Box, extentX=1.3919, extentY=6.4093, extentZ=6.603, center=<0.1 0.1 0.01>, rotation=<rot 0 -90.6915 -56.4659>}]
 root [{shapeType=Capsule, length=15, radius=10}]
}
constraints [
 {
  bone1Name=dyn_eye_l
  bone2Name=root
  swing1LimitDegrees=45
  swing1Motion=Locked
  swing2LimitDegrees=45
  swing2Motion=Free
  twistLimitDegrees=45
  twistMotion=Locked
 }
 {
  bone1Name=dyn_eye_r
  bone2Name=root
  swing1LimitDegrees=45
  swing1Motion=Locked
  swing2LimitDegrees=45
  swing2Motion=Free
  twistLimitDegrees=45
  twistMotion=Locked
 }
]
